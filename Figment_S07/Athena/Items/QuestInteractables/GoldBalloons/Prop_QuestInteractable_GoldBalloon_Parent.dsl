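; Parent Blueprint for destructible gold balloon quest props, which cannot be interacted with and always die when damaged.
; Construction scales the mesh and creates a dynamic material with tunable motion speeds.
; After a startup delay and playlist load, clients show gold if the match allows quests and the objective is incomplete, otherwise grey with a grey pop effect.
; On death, it sends the target-tag stat event to the instigating player's quest manager.
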(event EventBeginPlay
  ;[1 1 1] CLIENT AND SERVER - Ensure playlist is loaded
  (Utilities|FlowControl|Delay 5.0)
  (EventDispatchers|BindEventtoOnDied (EventDispatchers|CreateEvent :!event Custom|BalloonDied))
  (bind _asfort_game_state_athena (Utilities|Casting|CastToFortGameStateAthena (Game|GetGameState)))
  (if (Variables|/Script/FortniteGame.FortGameStateAthena|GetbPlaylistDataIsLoaded _asfort_game_state_athena)
    ;[1 1 1] Check Creative Mode
    (bind _cancompletequest (Class|AthenaQuestFunctionLibrary|AthenaCheckCreativeMode (Game|GetGameState)))
    (bind _output_get (Variables|Default|SetMatchTypeCanCompleteQuest _cancompletequest))
    (if _output_get
      (Networking|SwitchHasAuthority
        (:Remote
          ;[1 1 1] Not sure how to get this information without hard reference the quest def, athena doesn't seem to have a function for only checking the backend name
          (bind (_questisvalid _questcompleted _objectivecompleted) (Default|AthenaCheckQuestandObjectiveCompletion (Utilities|Array|Get(acopy) (Fortnite|GetLocalFortPlayerControllers true)) (Variables|Default|GetQuestDefinition) (Variables|Default|GetObjBackendName)))
          (if (and _questisvalid (not _objectivecompleted))
            ;[1 1 1] Turn Balloons grey in creative or playgrounds
            (CallFunction|SetColorGold)
            (else
              (CallFunction|SetColorGrey)))))
      (else
        (Networking|SwitchHasAuthority
          (:Remote
            (CallFunction|SetColorGrey)))))
    (else
      (Playlist|BindEventtoOnCurrentPlaylistLoaded _asfort_game_state_athena (delegate Custom|PlaylistLoaded)))))

(event Custom|PlaylistLoaded
  (params (PlaylistName :type Name) (PlaylistContextTags :type Gameplay%20Tag%20Container%20Structure%20%28by%20ref%29))
  ;[1 1 1] CLIENT AND SERVER - Ensure playlist is loaded
  ;[1 1 1] Check Creative Mode
  (bind _cancompletequest (Class|AthenaQuestFunctionLibrary|AthenaCheckCreativeMode (Game|GetGameState)))
  (bind _output_get (Variables|Default|SetMatchTypeCanCompleteQuest _cancompletequest))
  (if _output_get
    (Networking|SwitchHasAuthority
      (:Remote
        ;[1 1 1] Not sure how to get this information without hard reference the quest def, athena doesn't seem to have a function for only checking the backend name
        (bind (_questisvalid _questcompleted _objectivecompleted) (Default|AthenaCheckQuestandObjectiveCompletion (Utilities|Array|Get(acopy) (Fortnite|GetLocalFortPlayerControllers true)) (Variables|Default|GetQuestDefinition) (Variables|Default|GetObjBackendName)))
        (if (and _questisvalid (not _objectivecompleted))
          ;[1 1 1] Turn Balloons grey in creative or playgrounds
          (CallFunction|SetColorGold)
          (else
            (CallFunction|SetColorGrey)))))
    (else
      (Networking|SwitchHasAuthority
        (:Remote
          (CallFunction|SetColorGrey))))))

(event Custom|BalloonDied
  (params (DamagedActor :type Actor%20Object%20Reference) (Damage :type Float%20%28single-precision%29) (InstigatedBy :type Controller%20Object%20Reference) (DamageCauser :type Actor%20Object%20Reference) (HitLocation :type Vector) (FHitComponent :type Primitive%20Component%20Object%20Reference) (BoneName :type Name) (Momentum :type Vector))
  (bind _returnvalue (Fortnite|GetFortPlayerControllerfromActor InstigatedBy))
  (Utilities|IsValid _returnvalue
    (:Is%20Valid
      (Quest|SendComplexCustomStatEvent (Quests|GetQuestManager _returnvalue "Invalid") 0 "(GameplayTags=)" (Variables|Default|GetTargetTag) 1)
      (if true
        (Development|PrintString "Balloon Quest updated successfully " false false)
        (else
          (Development|PrintString "Player had this obj already" false false))))))

(fn BlueprintCanInteract
  (params (InteractingPawn :type Fort%20Pawn%20Object%20Reference) (InteractionBeingAttempted :type EInteractionBeingAttempted%20Enum :default "FirstInteraction") (InteractionType :type TInteractionType%20Enum :default "IT_NoInteraction"))
  (return false))

(fn SetColor_Gold
  (bind _dmi_balloon (Variables|Default|GetDMI_Balloon))
  (Rendering|Material|SetVectorParameterValue _dmi_balloon "Color_2" "(R=1.000000,G=0.756949,B=0.224885,A=1.000000)" :!class /Script/Engine.MaterialInstanceDynamic)
  (Rendering|Material|SetMaterial (Variables|Debug|GetStaticMeshComponent) 0 _dmi_balloon))

(fn SetColor_Grey
  (bind _dmi_balloon (Variables|Default|GetDMI_Balloon))
  (Rendering|Material|SetVectorParameterValue _dmi_balloon "Color_2" "(R=0.537255,G=0.803922,B=1.000000,A=1.000000)" :!class /Script/Engine.MaterialInstanceDynamic)
  (Rendering|Material|SetMaterial (Variables|Debug|GetStaticMeshComponent) 0 _dmi_balloon)
  (CallFunction|SetDeathParticles "/Figment_S07/Athena/Items/QuestInteractables/GoldBalloons/Art/P_GoldBalloon_GreyPop.P_GoldBalloon_GreyPop"))

(fn ShouldDie
  (params (Damage :type Float%20%28single-precision%29) (EventInstigator :type Controller%20Object%20Reference) (DamageCauser :type Actor%20Object%20Reference) (bIsTest :type Boolean) (EffectContext :type Gameplay%20Effect%20Context%20Handle%20Structure%20%28by%20ref%29))
  (return true))

(fn ConstructionScript
  (Transformation|SetWorldScale3D (Variables|Debug|GetStaticMeshComponent) (Variables|Default|GetBalloonScale))
  (bind _returnvalue (Rendering|Material|CreateDynamicMaterialInstance (Variables|Debug|GetStaticMeshComponent) 0 "/Figment_S07/Athena/Items/QuestInteractables/GoldBalloons/Art/MI_GoldBalloon_Gold.MI_GoldBalloon_Gold" "None" :!class /Script/Engine.PrimitiveComponent))
  (bind _output_get (Variables|Default|SetDMI_Balloon _returnvalue))
  (Utilities|FlowControl|Sequence
    (:then_0
      (Rendering|Material|SetScalarParameterValue _output_get "OverallRotationSpeed" (Variables|Default|GetOverallRotationSpeed) :!class /Script/Engine.MaterialInstanceDynamic))
    (:then_1
      (Rendering|Material|SetScalarParameterValue _output_get "SpeedBallon_2" (Variables|Default|GetSpeedBalloon_2) :!class /Script/Engine.MaterialInstanceDynamic))))
